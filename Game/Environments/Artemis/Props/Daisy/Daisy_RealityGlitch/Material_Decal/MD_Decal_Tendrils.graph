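# Translucent deferred-decal material for tendrils in the Daisy RealityGlitch prop set.
# A custom atan2 node builds polar UVs that scroll the tendril texture along its radius, masked by a time-pulsed radial falloff.
# Named reroutes define panning noise distortion, rotating four-color tint, sine-driven glitch flicker, noisy sphere-mask "Fresnel" and an actor-position random seed.
# Reroute links and final material output connections are not shown.
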Material /Game/Environments/Artemis/Props/Daisy/Daisy_RealityGlitch/Material_Decal/MD_Decal_Tendrils
graph {
 properties {
  bAutomaticallySetUsageInEditor=false
  blendMode=BLEND_Translucent
  materialDomain=MD_DeferredDecal
  pixelDepthOffsetMode=PDOM_Legacy
  shadingModel=MSM_Unlit
  usedShadingModels=Unlit
 }
 nodes [
  n1 = NamedRerouteUsage()
  n2 = Subtract(A=Add(A=Multiply(A=@n1) {constB=0.05}, B=TextureCoordinate() {uTiling=2, vTiling=2}), B=Constant() {r=1})
  MaterialExpressionComponentMask_15 = ComponentMask(None=@n2) {r=true}
  MaterialExpressionComponentMask_14 = ComponentMask(None=@n2) {g=true}
  n3 = Custom!(x=@MaterialExpressionComponentMask_15, y=@MaterialExpressionComponentMask_14) {
   code="return (atan2(x,y));"
   description=atan2
   outputType=CMOT_Float1
   inputs [
    {
     inputName=x
     input {expression=&self:MaterialExpressionComponentMask_15}
    }
    {
     inputName=y
     input {expression=&self:MaterialExpressionComponentMask_14}
    }
   ]
  }
  n4 = ComponentMask(None=@n3) {r=true}
  n5 = Constant() {r=3.2}
  n6 = Multiply(A=@n2, B=@n2)
  n7 = ComponentMask(None=@n6) {r=true}
  n8 = ComponentMask(None=@n6) {g=true}
  n9 = AppendVector(A=Frac(None=Divide(A=@n4, B=@n5)), B=SquareRoot(None=Add(A=@n7, B=@n8)))
  n10 = Multiply(A=Time(), B=Constant() {r=0.1})
  n11 = AppendVector(A=ComponentMask(None=@n9) {r=true}, B=ComponentMask(None=Panner(Coordinate=@n9, Time=@n10) {bFractionalPart=true, speedY=1}) {g=true})
  n12 = TextureCoordinate()
  n13 = MaterialFunctionCall(Vector or UVs=Subtract(A=Multiply(A=@n12) {constB=2})) {materialFunction=&VectorToRadialValue}
  n14 = Power(
   Base=Saturate(None=@n13.Linear Distance)
   Exp=LinearInterpolate(Alpha=MaterialFunctionCall() {materialFunction=&DebugTimeSine}.Result) {constA=20, constB=3}
  ) {constExponent=20}
  Add(
   A=Multiply(
    A=TextureSample(UVs=@n11) {samplerType=SAMPLERTYPE_LinearColor, texture=/Game/Environments/Artemis/Props/Daisy/Daisy_RealityGlitch/Textures/T_Tendrils}.RGB
    B=OneMinus(None=@n14)
   )
   B=NamedRerouteUsage()
  )
  n15 = TextureCoordinate() {uTiling=0.5, vTiling=0.5}
  NamedRerouteDeclaration(None=TextureSample(UVs=Panner(Coordinate=@n15) {speedX=-0.05, speedY=0.05}) {samplerType=SAMPLERTYPE_LinearGrayscale, texture=/Game/Weapons/FORT_Melee/Pickaxe_SnowFall_Female/FX/Snowfall_pick_Noise}.R) {name=Distortion, nodeColor=<rgba 1 0.7031 0 1>}
  n16 = TextureCoordinate()
  n17 = Multiply(A=Time()) {constB=4}
  n18 = Subtract(A=Multiply(A=Rotator(Coordinate=@n16, Time=@n17)) {constB=2})
  n19 = Multiply(A=TextureSample(UVs=MaterialFunctionCall(Vector or UVs=@n18) {materialFunction=&VectorToRadialValue}.Radial Coordinates) {texture=/Game/Environments/Artemis/Props/Daisy/Daisy_RealityGlitch/Textures/T_4colors}.RGB)
  NamedRerouteDeclaration(None=Multiply(A=Saturate(None=@n19))) {name=Color Tint, nodeColor=<rgba 1 0.8906 0 1>}
  n20 = Multiply(A=Time(), B=Constant() {r=0.15}) {constB=0.3}
  n21 = NamedRerouteUsage()
  n22 = Sine(None=Frac(None=Add(A=@n20, B=@n21)))
  n23 = Step(Y=NamedRerouteUsage()) {constX=0.5}
  n24 = Sine(None=Multiply(A=@n20, B=Multiply(A=@n23) {constB=0.6}) {constB=0.5})
  n25 = Add(A=Frac(None=Power(Base=@n22) {constExponent=302}), B=Frac(None=Power(Base=@n24) {constExponent=304}))
  NamedRerouteDeclaration(None=Saturate(None=Divide(A=@n25))) {name=Glitch, nodeColor=<rgba 0.2656 0 1 1>}
  n26 = TextureCoordinate()
  n27 = Constant2Vector() {g=0.5, r=0.5}
  n28 = Constant() {r=0.75}
  n29 = Constant() {r=0.25}
  n30 = TextureSample(UVs=Panner(Coordinate=TextureCoordinate() {uTiling=2, vTiling=2}) {speedX=-0.8, speedY=0.5}) {samplerType=SAMPLERTYPE_Grayscale, texture=/FortWater/Textures/LowResBlurredNoise}
  n31 = SphereMask(
   A=Add(A=SphereMask(A=@n26, B=@n27, Radius=@n28, Hardness=@n29), B=Multiply(A=@n30.RGB) {constB=0.05})
   B=Constant() {r=1}
   Radius=Constant() {r=1}
   Hardness=Constant()
  )
  n32 = Constant() {r=0.125}
  n33 = Constant() {r=0.25}
  n34 = NamedRerouteUsage()
  NamedRerouteDeclaration(None=Multiply(A=Power(Base=@n31), B=LinearInterpolate(A=@n32, B=@n33, Alpha=@n34))) {name=Fresnel, nodeColor=<rgba 0.2656 0 1 1>}
  n35 = MaterialFunctionCall(Float3=ActorPositionWS()) {materialFunction=&BreakOutFloat3Components}
  n36 = Add(A=Add(A=Abs(None=@n35.R), B=Abs(None=@n35.G)), B=Abs(None=@n35.B))
  NamedRerouteDeclaration(None=Frac(None=Divide(A=@n36) {constB=3})) {name=Random Seed 0-1, nodeColor=<rgba 0.8281 1 0 1>}
  NamedRerouteDeclaration(None=@n13.Radial Coordinates) {name=Radial Coordinates, nodeColor=<rgba 1 0.75 0 1>}
  NamedRerouteUsage()
  ScalarParameter() {parameterName=Metallic}
  ScalarParameter() {defaultValue=1, parameterName=Specular}
  ScalarParameter() {defaultValue=1, parameterName=Roughness}
 ]
}
